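# Physics asset for the Myth medium female body's dynamic skirt and coat bones, with capsules on legs, pelvis and spine.
# Eight three-bone skirt chains hang from spine_01 with limited joints, front and side ones via tiny MSTR bodies, while five coat bodies attach to spine_05.
# Front chains lock one swing axis and use boxes or long capsules on lower segments for wider coverage.
# All-free constraints also link some neighboring chains.
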PhysicsAsset /BRCosmetics/Characters/Player/Female/Medium/Bodies/F_MED_Myth/Meshes/F_MED_Myth_Dyn_skirt
bodies {
 calf_l [{shapeType=Capsule, length=52, radius=6.9375, center=<-25.4924 -2.8032 -0.0848>, rotation=<rot 86.2564 -180 -180>}]
 calf_r [{shapeType=Capsule, length=52, radius=6.9375, center=<25.4924 2.8032 0.0848>, rotation=<rot 86.2564 -180 -180>}]
 dyn_coat_bk [{shapeType=Capsule, length=6.5938, radius=4.7618, center=<5.8853 0 4.3>, rotation=<rot 90 0 0>}]
 dyn_coat_bk_l [{shapeType=Capsule, length=6.5938, radius=4.7618, center=<5.8853 0 4.3>, rotation=<rot 90 0 0>}]
 dyn_coat_bk_r [{shapeType=Capsule, length=6.5938, radius=4.7618, center=<5.8853 0 4.3>, rotation=<rot 90 0 0>}]
 dyn_coat_l [{shapeType=Capsule, length=6.5938, radius=4.7618, center=<5.8853 0 4.3>, rotation=<rot 90 0 0>}]
 dyn_coat_r [{shapeType=Capsule, length=6.5938, radius=4.7618, center=<5.8853 0 4.3>, rotation=<rot 90 0 0>}]
 dyn_skirt_bk_01 [{shapeType=Capsule, length=8.1389, radius=3.2167, center=<5.8853 0 4.3>, rotation=<rot 90 0 0>}]
 dyn_skirt_bk_02 [{shapeType=Capsule, length=6.5938, radius=4.7618, center=<5.8853 -0.0001 2.7>, rotation=<rot 90 0 0>}]
 dyn_skirt_bk_03 [{shapeType=Capsule, length=6.5938, radius=4.7618, center=<5.8853 0 1.5>, rotation=<rot 90 0 0>}]
 dyn_skirt_bk_l_01 [{shapeType=Capsule, length=11.6389, radius=2.8104, center=<5.8853 0 4.3>, rotation=<rot 90 0 0>}]
 dyn_skirt_bk_l_02 [{shapeType=Capsule, length=7.5312, radius=3.8243, center=<5.8853 -0.0001 2.7>, rotation=<rot 90 0 0>}]
 dyn_skirt_bk_l_03 [{shapeType=Capsule, length=6.0312, radius=5.3243, center=<5.8853 0 1.5>, rotation=<rot 90 0 0>}]
 dyn_skirt_bk_r_01 [{shapeType=Capsule, length=10.0938, radius=4.3556, center=<5.8853 0 4.3>, rotation=<rot 90 0 0>}]
 dyn_skirt_bk_r_02 [{shapeType=Capsule, length=7.5312, radius=3.8243, center=<5.8853 -0.0001 2.7>, rotation=<rot 90 0 0>}]
 dyn_skirt_bk_r_03 [{shapeType=Capsule, length=6.0312, radius=5.3243, center=<5.8853 0 1.5>, rotation=<rot 90 0 0>}]
 dyn_skirt_fr_01 [{shapeType=Capsule, length=7.6875, radius=3.6681, center=<5.8853 0 4.3>, rotation=<rot 90 0 0>}]
 dyn_skirt_fr_02 [{shapeType=Capsule, length=61.8487, radius=5.9891, shapeName=dyn_skirt_fr_02_capsule, center=<5.3223 -0.0348 4.0629>, rotation=<rot 0 0 90>}]
 dyn_skirt_fr_03 [{shapeType=Capsule, length=62.7821, radius=6.095, shapeName=dyn_skirt_fr_03_capsule, center=<5.2027 0.0631 3.0216>, rotation=<rot 0 0 90>}]
 dyn_skirt_fr_MSTR [{shapeType=Capsule, length=0.505, radius=0.505}]
 dyn_skirt_fr_l_01 [{shapeType=Capsule, length=7.4375, radius=3.9181, center=<5.8853 0 5.6>, rotation=<rot 90 0 0>}]
 dyn_skirt_fr_l_02 [{shapeType=Box, extentX=10.75, extentY=16.0625, extentZ=4.0625, center=<4.9147 -0.0005 0>, rotation=<rot 0 0 -8.615>}]
 dyn_skirt_fr_l_03 [{shapeType=Box, extentX=10.75, extentY=16.0625, extentZ=5.4375, center=<4.9147 -0.0005 0>, rotation=<rot 0 0 -2.8589>}]
 dyn_skirt_fr_l_MSTR [{shapeType=Capsule, length=0.505, radius=0.505}]
 dyn_skirt_fr_r_01 [{shapeType=Capsule, length=7.4375, radius=3.9181, center=<5.8853 0 5.6>, rotation=<rot 90 0 0>}]
 dyn_skirt_fr_r_02 [{shapeType=Box, extentX=10.75, extentY=16.0625, extentZ=4.0625, center=<4.9147 3.3556 -0.5085>, rotation=<rot 0 0 8.6151>}]
 dyn_skirt_fr_r_03 [{shapeType=Box, extentX=10.75, extentY=16.0625, extentZ=4.0625, center=<4.9147 3.3556 -0.5085>, rotation=<rot 0 0 8.6151>}]
 dyn_skirt_fr_r_MSTR [{shapeType=Capsule, length=0.505, radius=0.505}]
 dyn_skirt_l_01 [{shapeType=Capsule, length=9.7702, radius=3.7952, center=<5.8853 0 1.6252>, rotation=<rot 0 0 90>}]
 dyn_skirt_l_02 [{shapeType=Capsule, length=10.8082, radius=3.0454, shapeName=dyn_skirt_r_02_capsule, center=<5.5309 1.1044 1.8505>, rotation=<rot 0 0 90>}]
 dyn_skirt_l_03 [{shapeType=Capsule, length=11.5128, radius=2.9677, shapeName=dyn_skirt_r_03_capsule, center=<3.132 -0.1651 0.657>, rotation=<rot 0 0 90>}]
 dyn_skirt_l_MSTR [{shapeType=Capsule, length=0.505, radius=0.505}]
 dyn_skirt_r_01 [{shapeType=Capsule, length=5.996, radius=5.3595, center=<5.8853 0 4.3>, rotation=<rot 90 0 0>}]
 dyn_skirt_r_02 [{shapeType=Capsule, length=9.1706, radius=4.6831, shapeName=dyn_skirt_r_02_capsule, center=<5.5309 1.1044 2.4238>, rotation=<rot 0 0 90>}]
 dyn_skirt_r_03 [{shapeType=Capsule, length=10.2111, radius=4.2694, shapeName=dyn_skirt_r_03_capsule, center=<3.132 -0.1652 1.9034>, rotation=<rot 0 0 90>}]
 dyn_skirt_r_MSTR [{shapeType=Capsule, length=0.505, radius=0.505}]
 pelvis [{shapeType=Capsule, length=9.7188, radius=9.375, center=<-4.5613 -1.7298 0.0002>}]
 spine_01 [{shapeType=Capsule, length=2.875, radius=8.4375, center=<-0.9567 1.327 0>}]
 spine_05 [{shapeType=Capsule, length=2.875, radius=8.4375, center=<-0.9567 1.327 0>}]
 thigh_l [
  {shapeType=Capsule, length=33.2188, radius=8, center=<-20.3352 0.5458 0.9821>, rotation=<rot 85.6632 -180 -176.3707>}
  {shapeType=Capsule, length=4.8438, radius=6.5625, center=<-13.8281 0.632 -1.8468>, rotation=<rot 84.4141 180 180>}
 ]
 thigh_r [
  {shapeType=Capsule, length=33.2188, radius=8, center=<20.3352 -0.5458 -0.9821>, rotation=<rot 85.6632 -180 -176.3707>}
  {shapeType=Capsule, length=4.8438, radius=6.5625, center=<13.8281 -0.632 1.8468>, rotation=<rot 84.4141 180 180>}
 ]
}
constraints [
 {
  bone1Name=dyn_skirt_fr_01
  bone2Name=dyn_skirt_fr_MSTR
  swing1LimitDegrees=15
  swing1Motion=Locked
  swing2LimitDegrees=45
  twistLimitDegrees=45
  twistMotion=Locked
 }
 {
  bone1Name=dyn_skirt_fr_02
  bone2Name=dyn_skirt_fr_01
  swing1LimitDegrees=15
  swing1Motion=Locked
  swing2LimitDegrees=45
  twistLimitDegrees=45
 }
 {
  bone1Name=dyn_skirt_fr_03
  bone2Name=dyn_skirt_fr_02
  swing1LimitDegrees=15
  swing1Motion=Locked
  swing2LimitDegrees=45
  twistLimitDegrees=45
 }
 {
  bone1Name=dyn_skirt_fr_l_01
  bone2Name=dyn_skirt_fr_l_MSTR
  swing1LimitDegrees=5
  swing1Motion=Locked
  swing2LimitDegrees=30
  twistLimitDegrees=5
 }
 {
  bone1Name=dyn_skirt_fr_l_02
  bone2Name=dyn_skirt_fr_l_01
  swing1LimitDegrees=5
  swing1Motion=Locked
  swing2LimitDegrees=22
  twistLimitDegrees=5
 }
 {
  bone1Name=dyn_skirt_fr_l_03
  bone2Name=dyn_skirt_fr_l_02
  swing1LimitDegrees=15
  swing1Motion=Locked
  swing2LimitDegrees=45
  twistLimitDegrees=5
 }
 {bone1Name=dyn_skirt_l_01, bone2Name=dyn_skirt_l_MSTR, swing1LimitDegrees=5, swing2LimitDegrees=45, twistLimitDegrees=5}
 {bone1Name=dyn_skirt_l_02, bone2Name=dyn_skirt_l_01, swing1LimitDegrees=10, swing2LimitDegrees=45, twistLimitDegrees=5}
 {bone1Name=dyn_skirt_l_03, bone2Name=dyn_skirt_l_02, swing1LimitDegrees=30, swing2LimitDegrees=45, twistLimitDegrees=5}
 {bone1Name=dyn_skirt_bk_l_02, bone2Name=dyn_skirt_bk_l_01, swing1LimitDegrees=15, swing2LimitDegrees=45, twistLimitDegrees=5}
 {bone1Name=dyn_skirt_bk_l_03, bone2Name=dyn_skirt_bk_l_02, swing1LimitDegrees=15, swing2LimitDegrees=45, twistLimitDegrees=5}
 {bone1Name=dyn_skirt_bk_02, bone2Name=dyn_skirt_bk_01, swing1LimitDegrees=15, swing2LimitDegrees=45, twistLimitDegrees=5}
 {bone1Name=dyn_skirt_bk_03, bone2Name=dyn_skirt_bk_02, swing1LimitDegrees=15, swing2LimitDegrees=45, twistLimitDegrees=5}
 {bone1Name=dyn_skirt_bk_r_02, bone2Name=dyn_skirt_bk_r_01, swing1LimitDegrees=15, swing2LimitDegrees=45, twistLimitDegrees=5}
 {bone1Name=dyn_skirt_bk_r_03, bone2Name=dyn_skirt_bk_r_02, swing1LimitDegrees=15, swing2LimitDegrees=45, twistLimitDegrees=5}
 {bone1Name=dyn_skirt_r_01, bone2Name=dyn_skirt_r_MSTR, swing1LimitDegrees=5, swing2LimitDegrees=45, twistLimitDegrees=5}
 {bone1Name=dyn_skirt_r_02, bone2Name=dyn_skirt_r_01, swing1LimitDegrees=10, swing2LimitDegrees=45, twistLimitDegrees=5}
 {bone1Name=dyn_skirt_r_03, bone2Name=dyn_skirt_r_02, swing1LimitDegrees=15, swing2LimitDegrees=45, twistLimitDegrees=5}
 {
  bone1Name=dyn_skirt_fr_r_01
  bone2Name=dyn_skirt_fr_r_MSTR
  swing1LimitDegrees=5
  swing1Motion=Locked
  swing2LimitDegrees=45
  twistLimitDegrees=5
 }
 {
  bone1Name=dyn_skirt_fr_r_02
  bone2Name=dyn_skirt_fr_r_01
  swing1LimitDegrees=5
  swing1Motion=Locked
  swing2LimitDegrees=45
  twistLimitDegrees=15
 }
 {
  bone1Name=dyn_skirt_fr_r_03
  bone2Name=dyn_skirt_fr_r_02
  swing1LimitDegrees=15
  swing1Motion=Locked
  swing2LimitDegrees=45
  twistLimitDegrees=15
 }
 {
  bone1Name=dyn_skirt_fr_MSTR
  bone2Name=spine_01
  swing1LimitDegrees=45
  swing1Motion=Free
  swing2LimitDegrees=45
  swing2Motion=Free
  twistLimitDegrees=45
  twistMotion=Free
 }
 {
  bone1Name=dyn_skirt_fr_l_MSTR
  bone2Name=spine_01
  swing1LimitDegrees=45
  swing1Motion=Free
  swing2LimitDegrees=45
  swing2Motion=Free
  twistLimitDegrees=45
  twistMotion=Free
 }
 {
  bone1Name=dyn_skirt_l_MSTR
  bone2Name=spine_01
  swing1LimitDegrees=45
  swing1Motion=Free
  swing2LimitDegrees=45
  swing2Motion=Free
  twistLimitDegrees=45
  twistMotion=Free
 }
 {bone1Name=dyn_skirt_bk_l_01, bone2Name=spine_01, swing1LimitDegrees=15, swing2LimitDegrees=45, twistLimitDegrees=5}
 {bone1Name=dyn_skirt_bk_01, bone2Name=spine_01, swing1LimitDegrees=15, swing2LimitDegrees=45, twistLimitDegrees=5}
 {bone1Name=dyn_skirt_bk_r_01, bone2Name=spine_01, swing1LimitDegrees=15, swing2LimitDegrees=45, twistLimitDegrees=5}
 {
  bone1Name=dyn_skirt_r_MSTR
  bone2Name=spine_01
  swing1LimitDegrees=45
  swing1Motion=Free
  swing2LimitDegrees=45
  swing2Motion=Free
  twistLimitDegrees=45
  twistMotion=Free
 }
 {
  bone1Name=dyn_skirt_fr_r_MSTR
  bone2Name=spine_01
  swing1LimitDegrees=45
  swing1Motion=Free
  swing2LimitDegrees=45
  swing2Motion=Free
  twistLimitDegrees=45
  twistMotion=Free
 }
 {
  bone1Name=spine_01
  bone2Name=pelvis
  swing1LimitDegrees=45
  swing1Motion=Free
  swing2LimitDegrees=45
  swing2Motion=Free
  twistLimitDegrees=45
  twistMotion=Free
 }
 {
  bone1Name=thigh_l
  bone2Name=pelvis
  swing1LimitDegrees=45
  swing1Motion=Free
  swing2LimitDegrees=45
  swing2Motion=Free
  twistLimitDegrees=45
  twistMotion=Free
 }
 {
  bone1Name=thigh_r
  bone2Name=pelvis
  swing1LimitDegrees=45
  swing1Motion=Free
  swing2LimitDegrees=45
  swing2Motion=Free
  twistLimitDegrees=45
  twistMotion=Free
 }
 {
  bone1Name=calf_r
  bone2Name=thigh_r
  swing1LimitDegrees=45
  swing1Motion=Free
  swing2LimitDegrees=45
  swing2Motion=Free
  twistLimitDegrees=45
  twistMotion=Free
 }
 {
  bone1Name=calf_l
  bone2Name=thigh_l
  swing1LimitDegrees=45
  swing1Motion=Free
  swing2LimitDegrees=45
  swing2Motion=Free
  twistLimitDegrees=45
  twistMotion=Free
 }
 {bone1Name=dyn_coat_r, bone2Name=spine_05, swing1LimitDegrees=15, swing2LimitDegrees=45, twistLimitDegrees=5}
 {bone1Name=dyn_coat_bk_r, bone2Name=spine_05, swing1LimitDegrees=15, swing2LimitDegrees=45, twistLimitDegrees=5}
 {bone1Name=dyn_coat_bk, bone2Name=spine_05, swing1LimitDegrees=15, swing2LimitDegrees=45, twistLimitDegrees=5}
 {bone1Name=dyn_coat_bk_l, bone2Name=spine_05, swing1LimitDegrees=15, swing2LimitDegrees=45, twistLimitDegrees=5}
 {bone1Name=dyn_coat_l, bone2Name=spine_05, swing1LimitDegrees=15, swing2LimitDegrees=45, twistLimitDegrees=5}
 {
  bone1Name=spine_05
  bone2Name=spine_01
  swing1LimitDegrees=45
  swing1Motion=Free
  swing2LimitDegrees=45
  swing2Motion=Free
  twistLimitDegrees=45
  twistMotion=Free
 }
 {
  bone1Name=dyn_skirt_fr_l_01
  bone2Name=dyn_skirt_l_01
  swing1LimitDegrees=45
  swing1Motion=Free
  swing2LimitDegrees=45
  swing2Motion=Free
  twistLimitDegrees=45
  twistMotion=Free
 }
 {
  bone1Name=dyn_skirt_fr_r_01
  bone2Name=dyn_skirt_r_01
  swing1LimitDegrees=45
  swing1Motion=Free
  swing2LimitDegrees=45
  swing2Motion=Free
  twistLimitDegrees=45
  twistMotion=Free
 }
 {
  bone1Name=dyn_skirt_fr_r_02
  bone2Name=dyn_skirt_r_02
  swing1LimitDegrees=45
  swing1Motion=Free
  swing2LimitDegrees=45
  swing2Motion=Free
  twistLimitDegrees=45
  twistMotion=Free
 }
 {
  bone1Name=dyn_skirt_fr_r_03
  bone2Name=dyn_skirt_r_03
  swing1LimitDegrees=45
  swing1Motion=Free
  swing2LimitDegrees=45
  swing2Motion=Free
  twistLimitDegrees=45
  twistMotion=Free
 }
 {
  bone1Name=dyn_skirt_l_02
  bone2Name=dyn_skirt_bk_l_02
  swing1LimitDegrees=45
  swing1Motion=Free
  swing2LimitDegrees=45
  swing2Motion=Free
  twistLimitDegrees=45
  twistMotion=Free
 }
 {
  bone1Name=dyn_skirt_l_03
  bone2Name=dyn_skirt_bk_l_03
  swing1LimitDegrees=45
  swing1Motion=Free
  swing2LimitDegrees=45
  swing2Motion=Free
  twistLimitDegrees=45
  twistMotion=Free
 }
 {
  bone1Name=dyn_skirt_fr_r_03
  bone2Name=dyn_skirt_fr_03
  swing1LimitDegrees=45
  swing1Motion=Free
  swing2LimitDegrees=45
  swing2Motion=Free
  twistLimitDegrees=45
  twistMotion=Free
 }
 {
  bone1Name=dyn_skirt_fr_l_03
  bone2Name=dyn_skirt_fr_03
  swing1LimitDegrees=45
  swing1Motion=Free
  swing2LimitDegrees=45
  swing2Motion=Free
  twistLimitDegrees=45
  twistMotion=Free
 }
]
