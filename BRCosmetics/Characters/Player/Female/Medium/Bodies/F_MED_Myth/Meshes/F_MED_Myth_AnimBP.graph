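# Post-process style AnimBlueprint for the F_MED_Myth female body, adding secondary motion to its skirt and bow strings.
# It copies the parent mesh pose, constrains front skirt bones to the thighs, then runs RigidBody skirt simulation.
# Crouch-aware bone offsets and weighted multi-bone constraints seat the strings on the skirt before AnimDynamics adds sway.
# The EventGraph sets skirt gravity for swim-sprinting, skydiving and parachuting, and flags crouch-walking.
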AnimBlueprint /BRCosmetics/Characters/Player/Female/Medium/Bodies/F_MED_Myth/Meshes/F_MED_Myth_AnimBP
target_skeleton=&/F_MED_Myth_Skeleton
graphs [
 AnimGraph {nodes [
   n1 = "Transform (Modify) Bone - Bone: dyn_bow_string_l_01"(Alpha=Get thigh_rotation_l() -> (thigh_rotation_l:double), Translation="0,0,0", Rotation="5,0,0", Scale="1,1,1") {
    rotationMode=BMM_Additive
    rotationSpace=BCS_BoneSpace
    translationMode=BMM_Additive
    translationSpace=BCS_BoneSpace
    alphaBoolBlend {blendOption=Linear}
    alphaScaleBiasClamp {interpSpeedIncreasing=40}
    boneToModify {boneName=dyn_bow_string_l_01}
    rotation {pitch=5}
   }
   "Transform (Modify) Bone - Bone: dyn_bow_string_r_01"(ComponentPose=@n1, Alpha=Get thigh rotation_r() -> (thigh rotation_r:double), Translation="0,0,0", Rotation="5,0,0", Scale="1,1,1") {
    rotationMode=BMM_Additive
    rotationSpace=BCS_BoneSpace
    translationMode=BMM_Additive
    translationSpace=BCS_BoneSpace
    alphaBoolBlend {blendOption=Linear}
    alphaScaleBiasClamp {bInterpResult=true, interpSpeedDecreasing=40, interpSpeedIncreasing=40}
    boneToModify {boneName=dyn_bow_string_r_01}
    rotation {pitch=5}
   }
   n2 = "Constraint - dyn_skirt_fr_l_MSTR (R)"(ComponentPose=Local To Component(LocalPose=Copy Pose From Mesh() {bUseAttachedParent=true}), Alpha=1, ConstraintWeights_0=1) {
    alphaBoolBlend {blendOption=Linear}
    boneToModify {boneName=dyn_skirt_fr_l_MSTR}
    constraintSetup [{transformType=Rotation, targetBone {boneName=thigh_l}}]
    constraintWeights [1]
   }
   n3 = "Constraint - dyn_skirt_fr_r_MSTR (R)"(ComponentPose=@n2, Alpha=1, ConstraintWeights_0=1) {
    alphaBoolBlend {blendOption=Linear}
    boneToModify {boneName=dyn_skirt_fr_r_MSTR}
    constraintSetup [{transformType=Rotation, targetBone {boneName=thigh_r}}]
    constraintWeights [1]
   }
   n4 = RigidBody!(ComponentPose=@n3, Alpha, OverrideWorldGravity:Vector=Get gravity_override() -> (gravity_override:Vector), ExternalForce) -> (Pose) {
    bClampLinearTranslationLimitToRefPose=true
    bOverrideWorldGravity=true
    overridePhysicsAsset=&/F_MED_Myth_Dyn_skirt
    simulationSpace=BaseBoneSpace
    alphaBoolBlend {blendOption=Linear}
    baseBoneRef {boneName=spine_01}
    componentAppliedLinearAccClamp=<3000 3000 3000>
    componentLinearAccScale=<0.4 0.3 0.3>
    componentLinearVelScale=<0.4 0.3 0.3>
    overrideWorldGravity {z=-980}
    simSpaceSettings {
     maxAngularAcceleration=500
     maxAngularVelocity=3
     maxLinearAcceleration=700
     maxLinearVelocity=500
     velocityScaleZ=0
     worldAlpha=1
     externalLinearDragV=<1 1>
    }
   }
   n5 = "Transform (Modify) Bone - Bone: dyn_skirt_fr_l_MSTR"!(ComponentPose=@n4, bAlphaBoolEnabled:bool=Get bIsCrouching(), Translation="0,0,7", Rotation="0,-5,0", Scale="1,1.3,1") -> (Pose) {
    alphaInputType=Bool
    rotationMode=BMM_Additive
    rotationSpace=BCS_BoneSpace
    scaleMode=BMM_Additive
    translationMode=BMM_Additive
    translationSpace=BCS_BoneSpace
    alphaBoolBlend {blendInTime=0.1, blendOption=Linear, blendOutTime=0.1}
    alphaScaleBiasClamp {interpSpeedIncreasing=40}
    boneToModify {boneName=dyn_skirt_fr_l_MSTR}
    rotation {yaw=-5}
    scale {y=1.3}
    translation {z=7}
   }
   n6 = Get bIsCrouching()
   n7 = "Constraint - dyn_skirt_l_01 (RR)"!(ComponentPose=@n5, bAlphaBoolEnabled:bool=@n6, ConstraintWeights_0=0.5, ConstraintWeights_1:float=0.5) -> (Pose) {
    alphaInputType=Bool
    alphaBoolBlend {blendInTime=0.2, blendOption=Linear, blendOutTime=0.1}
    alphaScaleBiasClamp {bInterpResult=true, interpSpeedDecreasing=20, interpSpeedIncreasing=20}
    boneToModify {boneName=dyn_skirt_l_01}
    constraintSetup [{transformType=Rotation, targetBone {boneName=dyn_skirt_fr_l_01}}, {transformType=Rotation, targetBone {boneName=dyn_skirt_bk_l_01}}]
    constraintWeights [1,1]
   }
   n8 = "Transform (Modify) Bone - Bone: dyn_skirt_l_01"!(ComponentPose=@n7, bAlphaBoolEnabled:bool=@n6, Translation="5,0,0", Rotation="20,0,0", Scale="1,1,1") -> (Pose) {
    alphaInputType=Bool
    rotationMode=BMM_Additive
    scaleMode=BMM_Additive
    translationMode=BMM_Additive
    alphaBoolBlend {blendInTime=0.1, blendOption=Linear, blendOutTime=0.1}
    alphaScaleBiasClamp {interpSpeedIncreasing=40}
    boneToModify {boneName=dyn_skirt_l_01}
    rotation {pitch=20}
    translation {x=5}
   }
   n9 = "Transform (Modify) Bone - Bone: dyn_skirt_fr_l_MSTR"(ComponentPose=@n8, bAlphaBoolEnabled=True, Translation="0,0,3", Rotation="0,0,0", Scale="1,1,1") {
    alphaInputType=Bool
    rotationMode=BMM_Additive
    rotationSpace=BCS_BoneSpace
    scaleMode=BMM_Additive
    scaleSpace=BCS_BoneSpace
    translationMode=BMM_Additive
    translationSpace=BCS_BoneSpace
    alphaBoolBlend {blendInTime=0.1, blendOption=Linear, blendOutTime=0.1}
    alphaScaleBiasClamp {interpSpeedIncreasing=40}
    boneToModify {boneName=dyn_skirt_fr_l_MSTR}
    translation {z=3}
   }
   n10 = "Transform (Modify) Bone - Bone: dyn_skirt_l_MSTR"!(ComponentPose=@n9, bAlphaBoolEnabled:bool=NOT Boolean(A=Get bIsCrouching()), Translation="0,0,2", Rotation="5,0,0", Scale="1,1,1") -> (Pose) {
    alphaInputType=Bool
    rotationMode=BMM_Additive
    rotationSpace=BCS_BoneSpace
    scaleMode=BMM_Additive
    scaleSpace=BCS_BoneSpace
    translationMode=BMM_Additive
    translationSpace=BCS_BoneSpace
    alphaBoolBlend {blendInTime=0.1, blendOption=Linear, blendOutTime=0.1}
    alphaScaleBiasClamp {interpSpeedIncreasing=40}
    boneToModify {boneName=dyn_skirt_l_MSTR}
    rotation {pitch=5}
    translation {z=2}
   }
   n11 = "Transform (Modify) Bone - Bone: dyn_skirt_r_MSTR"!(ComponentPose=@n10, bAlphaBoolEnabled:bool=True, Translation="0,0,2", Rotation="5,0,0", Scale="1,1,1") -> (Pose) {
    alphaInputType=Bool
    rotationMode=BMM_Additive
    rotationSpace=BCS_BoneSpace
    scaleMode=BMM_Additive
    scaleSpace=BCS_BoneSpace
    translationMode=BMM_Additive
    translationSpace=BCS_BoneSpace
    alphaBoolBlend {blendInTime=0.1, blendOption=Linear, blendOutTime=0.1}
    alphaScaleBiasClamp {interpSpeedIncreasing=40}
    boneToModify {boneName=dyn_skirt_r_MSTR}
    rotation {pitch=5}
    translation {z=2}
   }
   n12 = "Constraint - dyn_bow_string_l_01 (RRRRRRRR)"!(
    ComponentPose=@n11
    Alpha=1
    ConstraintWeights_0=0.045
    ConstraintWeights_1:float=0.1077
    ConstraintWeights_2:float=0.2797
    ConstraintWeights_3:float=0.0544
    ConstraintWeights_4:float=0.1131
    ConstraintWeights_5:float=0.0522
    ConstraintWeights_6:float=0.0203
    ConstraintWeights_7:float=0.3275
   ) -> (Pose) {
    alphaBoolBlend {blendOption=Linear}
    boneToModify {boneName=dyn_bow_string_l_01}
    constraintSetup [
     {transformType=Rotation, targetBone {boneName=dyn_skirt_fr_r_01}}
     {transformType=Rotation, targetBone {boneName=dyn_skirt_fr_01}}
     {transformType=Rotation, targetBone {boneName=dyn_skirt_fr_02}}
     {transformType=Rotation, targetBone {boneName=dyn_skirt_fr_03}}
     {transformType=Rotation, targetBone {boneName=dyn_skirt_fr_l_01}}
     {transformType=Rotation, targetBone {boneName=dyn_skirt_fr_l_02}}
     {transformType=Rotation, targetBone {boneName=spine_02}}
     {transformType=Rotation, targetBone {boneName=pelvis}}
    ]
    constraintWeights [1,1,1,1,1,1,1,1]
   }
   n13 = NOT Boolean(A=Get bIsCrouching())
   n14 = "Constraint - dyn_bow_string_r_02 (PPPPPPPP)"!(
    ComponentPose=@n12
    bAlphaBoolEnabled:bool=@n13
    ConstraintWeights_0=0.0149
    ConstraintWeights_1:float=0.131
    ConstraintWeights_2:float=0.1361
    ConstraintWeights_3:float=0.0938
    ConstraintWeights_4:float=0.3616
    ConstraintWeights_5:float=0.1527
    ConstraintWeights_6:float=0.0164
    ConstraintWeights_7:float=0.0935
   ) -> (Pose) {
    alphaInputType=Bool
    alphaBoolBlend {blendOption=Linear}
    boneToModify {boneName=dyn_bow_string_r_02}
    constraintSetup [
     {targetBone {boneName=dyn_skirt_fr_r_03}}
     {targetBone {boneName=dyn_skirt_fr_r_02}}
     {targetBone {boneName=dyn_skirt_fr_r_01}}
     {targetBone {boneName=dyn_skirt_fr_01}}
     {targetBone {boneName=dyn_skirt_fr_02}}
     {targetBone {boneName=dyn_skirt_fr_03}}
     {targetBone {boneName=spine_02}}
     {targetBone {boneName=pelvis}}
    ]
    constraintWeights [1,1,1,1,1,1,1,1]
   }
   n15 = "Constraint - dyn_bow_string_r_03 (PPPPPPPP)"!(
    ComponentPose=@n14
    bAlphaBoolEnabled:bool=@n13
    ConstraintWeights_0=0.0295
    ConstraintWeights_1:float=0.1029
    ConstraintWeights_2:float=0.102
    ConstraintWeights_3:float=0.0752
    ConstraintWeights_4:float=0.334
    ConstraintWeights_5:float=0.2284
    ConstraintWeights_6:float=0.0253
    ConstraintWeights_7:float=0.1027
   ) -> (Pose) {
    alphaInputType=Bool
    alphaBoolBlend {blendOption=Linear}
    boneToModify {boneName=dyn_bow_string_r_03}
    constraintSetup [
     {targetBone {boneName=dyn_bow_string_r_03}}
     {targetBone {boneName=dyn_skirt_fr_r_02}}
     {targetBone {boneName=dyn_skirt_fr_r_01}}
     {targetBone {boneName=dyn_skirt_fr_01}}
     {targetBone {boneName=dyn_skirt_fr_02}}
     {targetBone {boneName=dyn_skirt_fr_03}}
     {targetBone {boneName=spine_02}}
     {targetBone {boneName=pelvis}}
    ]
    constraintWeights [1,1,1,1,1,1,1,1]
   }
   n16 = "Constraint - dyn_bow_string_r_01 (RRRRRRRR)"!(
    ComponentPose=@n15
    Alpha=1
    ConstraintWeights_0=0.0408
    ConstraintWeights_1:float=0.0887
    ConstraintWeights_2:float=0.1039
    ConstraintWeights_3:float=0.2819
    ConstraintWeights_4:float=0.0657
    ConstraintWeights_5:float=0.0521
    ConstraintWeights_6:float=0.0216
    ConstraintWeights_7:float=0.3454
   ) -> (Pose) {
    alphaBoolBlend {blendOption=Linear}
    boneToModify {boneName=dyn_bow_string_r_01}
    constraintSetup [
     {transformType=Rotation, targetBone {boneName=dyn_skirt_fr_r_02}}
     {transformType=Rotation, targetBone {boneName=dyn_skirt_fr_r_01}}
     {transformType=Rotation, targetBone {boneName=dyn_skirt_fr_01}}
     {transformType=Rotation, targetBone {boneName=dyn_skirt_fr_02}}
     {transformType=Rotation, targetBone {boneName=dyn_skirt_fr_03}}
     {transformType=Rotation, targetBone {boneName=dyn_skirt_fr_l_01}}
     {transformType=Rotation, targetBone {boneName=dyn_skirt_fr_l_02}}
     {transformType=Rotation, targetBone {boneName=pelvis}}
    ]
    constraintWeights [1,1,1,1,1,1,1,1]
   }
   n17 = "Constraint - dyn_bow_string_l_02 (PPPPPPPPPPPP)"!(
    ComponentPose=@n16
    bAlphaBoolEnabled:bool=@n13
    ConstraintWeights_0=0.0016
    ConstraintWeights_1:float=0.0157
    ConstraintWeights_2:float=0.0744
    ConstraintWeights_3:float=0.2997
    ConstraintWeights_4:float=0.1351
    ConstraintWeights_5:float=0.1315
    ConstraintWeights_6:float=0.006
    ConstraintWeights_7:float=0.2346
    ConstraintWeights_8:float=0.009
    ConstraintWeights_9:float=0.0001
    ConstraintWeights_10:float=0.0904
    ConstraintWeights_11:float=0.0019
   ) -> (Pose) {
    alphaInputType=Bool
    alphaBoolBlend {blendOption=Linear}
    boneToModify {boneName=dyn_bow_string_l_02}
    constraintSetup [
     {targetBone {boneName=dyn_skirt_fr_r_02}}
     {targetBone {boneName=dyn_skirt_fr_r_01}}
     {targetBone {boneName=dyn_skirt_fr_01}}
     {targetBone {boneName=dyn_skirt_fr_02}}
     {targetBone {boneName=dyn_skirt_fr_03}}
     {targetBone {boneName=dyn_skirt_fr_l_01}}
     {targetBone {boneName=dyn_skirt_fr_l_03}}
     {targetBone {boneName=dyn_skirt_fr_l_02}}
     {targetBone {boneName=spine_02}}
     {targetBone {boneName=dyn_skirt_l_03}}
     {targetBone {boneName=pelvis}}
     {targetBone {boneName=dyn_skirt_l_02}}
    ]
    constraintWeights [1,1,1,1,1,1,1,1,1,1,1,1]
   }
   n18 = "Constraint - dyn_bow_string_l_03 (PPPPPPPPPPPP)"!(
    ComponentPose=@n17
    bAlphaBoolEnabled:bool=@n13
    ConstraintWeights_0=0.003
    ConstraintWeights_1:float=0.0026
    ConstraintWeights_2:float=0.0559
    ConstraintWeights_3:float=0.2797
    ConstraintWeights_4:float=0.2116
    ConstraintWeights_5:float=0.0756
    ConstraintWeights_6:float=0.0235
    ConstraintWeights_7:float=0.2693
    ConstraintWeights_8:float=0.0121
    ConstraintWeights_9:float=0.0012
    ConstraintWeights_10:float=0.0632
    ConstraintWeights_11:float=0.0023
   ) -> (Pose) {
    alphaInputType=Bool
    alphaBoolBlend {blendOption=Linear}
    boneToModify {boneName=dyn_bow_string_l_03}
    constraintSetup [
     {targetBone {boneName=dyn_skirt_fr_r_02}}
     {targetBone {boneName=dyn_skirt_fr_r_01}}
     {targetBone {boneName=dyn_skirt_fr_01}}
     {targetBone {boneName=dyn_skirt_fr_02}}
     {targetBone {boneName=dyn_skirt_fr_03}}
     {targetBone {boneName=dyn_skirt_fr_l_01}}
     {targetBone {boneName=dyn_skirt_fr_l_03}}
     {targetBone {boneName=dyn_skirt_fr_l_02}}
     {targetBone {boneName=spine_02}}
     {targetBone {boneName=dyn_skirt_l_03}}
     {targetBone {boneName=pelvis}}
     {targetBone {boneName=dyn_skirt_l_02}}
    ]
    constraintWeights [1,1,1,1,1,1,1,1,1,1,1,1]
   }
   n19 = "Transform (Modify) Bone - Bone: dyn_bow_string_l_01"!(ComponentPose=@n18, bAlphaBoolEnabled:bool=True, Translation="0,0,1", Rotation="10,0,0", Scale="1,1,1") -> (Pose) {
    alphaInputType=Bool
    rotationMode=BMM_Additive
    rotationSpace=BCS_BoneSpace
    scaleMode=BMM_Additive
    scaleSpace=BCS_BoneSpace
    translationMode=BMM_Additive
    translationSpace=BCS_BoneSpace
    alphaBoolBlend {blendInTime=0.1, blendOption=Linear, blendOutTime=0.1}
    alphaScaleBiasClamp {interpSpeedIncreasing=40}
    boneToModify {boneName=dyn_bow_string_l_01}
    rotation {pitch=10}
    translation {z=1}
   }
   n20 = "Transform (Modify) Bone - Bone: dyn_bow_string_r_01"!(ComponentPose=@n19, bAlphaBoolEnabled:bool=True, Translation="0,0,1", Rotation="10,0,0", Scale="1,1,1") -> (Pose) {
    alphaInputType=Bool
    rotationMode=BMM_Additive
    rotationSpace=BCS_BoneSpace
    scaleMode=BMM_Additive
    scaleSpace=BCS_BoneSpace
    translationMode=BMM_Additive
    translationSpace=BCS_BoneSpace
    alphaBoolBlend {blendInTime=0.1, blendOption=Linear, blendOutTime=0.1}
    alphaScaleBiasClamp {interpSpeedIncreasing=40}
    boneToModify {boneName=dyn_bow_string_r_01}
    rotation {pitch=10}
    translation {z=1}
   }
   n21 = "Transform (Modify) Bone - Bone: dyn_bow_string_l_01"(ComponentPose=@n20, bAlphaBoolEnabled=True, Translation="0,0,2", Rotation="30,0,0", Scale="1,1,1") {
    alphaInputType=Bool
    rotationMode=BMM_Additive
    rotationSpace=BCS_BoneSpace
    scaleMode=BMM_Additive
    scaleSpace=BCS_BoneSpace
    translationMode=BMM_Additive
    translationSpace=BCS_BoneSpace
    alphaBoolBlend {blendInTime=0.1, blendOption=Linear, blendOutTime=0.1}
    alphaScaleBiasClamp {interpSpeedIncreasing=40}
    boneToModify {boneName=dyn_bow_string_l_01}
    rotation {pitch=30}
    translation {z=2}
   }
   n22 = "Transform (Modify) Bone - Bone: dyn_bow_string_r_01"(ComponentPose=@n21, bAlphaBoolEnabled=True, Translation="0,0,2", Rotation="20,0,0", Scale="1,1,1") {
    alphaInputType=Bool
    rotationMode=BMM_Additive
    rotationSpace=BCS_BoneSpace
    scaleMode=BMM_Additive
    scaleSpace=BCS_BoneSpace
    translationMode=BMM_Additive
    translationSpace=BCS_BoneSpace
    alphaBoolBlend {blendInTime=0.1, blendOption=Linear, blendOutTime=0.1}
    alphaScaleBiasClamp {interpSpeedIncreasing=40}
    boneToModify {boneName=dyn_bow_string_r_01}
    rotation {pitch=20}
    translation {z=2}
   }
   n23 = Get CrouchWalking() -> (CrouchWalking:bool)
   n24 = "Transform (Modify) Bone - Bone: dyn_bow_string_l_01"(ComponentPose=@n22, bAlphaBoolEnabled=@n23, Translation="0,0,2", Rotation="10,0,0", Scale="1,1,1") {
    alphaInputType=Bool
    rotationMode=BMM_Additive
    rotationSpace=BCS_BoneSpace
    scaleMode=BMM_Additive
    scaleSpace=BCS_BoneSpace
    translationMode=BMM_Additive
    translationSpace=BCS_BoneSpace
    alphaBoolBlend {blendInTime=0.1, blendOption=Linear, blendOutTime=0.1}
    alphaScaleBiasClamp {interpSpeedIncreasing=40}
    boneToModify {boneName=dyn_bow_string_l_01}
    rotation {pitch=10}
    translation {z=2}
   }
   n25 = "Transform (Modify) Bone - Bone: dyn_bow_string_r_01"(ComponentPose=@n24, bAlphaBoolEnabled=@n23, Translation="0,0,2", Rotation="10,0,0", Scale="1,1,1") {
    alphaInputType=Bool
    rotationMode=BMM_Additive
    rotationSpace=BCS_BoneSpace
    scaleMode=BMM_Additive
    scaleSpace=BCS_BoneSpace
    translationMode=BMM_Additive
    translationSpace=BCS_BoneSpace
    alphaBoolBlend {blendInTime=0.1, blendOption=Linear, blendOutTime=0.1}
    alphaScaleBiasClamp {interpSpeedIncreasing=40}
    boneToModify {boneName=dyn_bow_string_r_01}
    rotation {pitch=10}
    translation {z=2}
   }
   n26 = SelectFloat(A=0.65, B=1, bPickA=NOT Boolean(A=Get bIsCrouching()))
   n27 = Anim Dynamics - Bone: dyn_bow_string_l_02(ComponentPose=@n25, Alpha=@n26, ExternalForce="0,0,0") {
    angularDampingOverride=0.9999
    angularSpringConstant=75
    bAngularSpring=true
    bGravityOverrideInSimSpace=true
    bOverrideAngularDamping=true
    simulationSpace=RootRelative
    alphaBoolBlend {blendOption=Linear}
    boundBone {boneName=dyn_bow_string_l_02}
    componentLinearAccScale=<1 1 1>
    componentLinearVelScale=<1 1 1>
    physicsBodyDefinitions [{
      sphereCollisionRadius=0
      boundBone {boneName=dyn_bow_string_l_02}
      boxExtents=<20 20 10>
      constraintSetup {angularLimitsMax=<25 10 25>, angularLimitsMin=<-25 -25 -25>, angularTarget=<0 0 0>}
      localJointOffset=<2 0 0>
     }]
   }
   n28 = Anim Dynamics - Bone: dyn_bow_string_l_03(ComponentPose=@n27, Alpha=@n26, ExternalForce="0,0,0") {
    angularDampingOverride=0.9999
    angularSpringConstant=75
    bAngularSpring=true
    bGravityOverrideInSimSpace=true
    bOverrideAngularDamping=true
    simulationSpace=RootRelative
    alphaBoolBlend {blendOption=Linear}
    boundBone {boneName=dyn_bow_string_l_03}
    componentLinearAccScale=<1 1 1>
    componentLinearVelScale=<1 1 1>
    physicsBodyDefinitions [{
      sphereCollisionRadius=0
      boundBone {boneName=dyn_bow_string_l_03}
      boxExtents=<20 20 10>
      constraintSetup {angularLimitsMax=<25 10 25>, angularLimitsMin=<-25 -25 -25>, angularTarget=<0 0 0>}
      localJointOffset=<2 0 0>
     }]
   }
   n29 = Anim Dynamics - Bone: dyn_bow_string_r_02(ComponentPose=@n28, Alpha=@n26, ExternalForce="0,0,0") {
    angularDampingOverride=0.9999
    angularSpringConstant=80
    bAngularSpring=true
    bGravityOverrideInSimSpace=true
    bOverrideAngularDamping=true
    bUseGravityOverride=true
    simulationSpace=RootRelative
    alphaBoolBlend {blendOption=Linear}
    boundBone {boneName=dyn_bow_string_r_02}
    componentLinearAccScale=<1 1 1>
    componentLinearVelScale=<1 1 1>
    gravityOverride {z=-980}
    physicsBodyDefinitions [{
      sphereCollisionRadius=0
      boundBone {boneName=dyn_bow_string_r_02}
      boxExtents=<20 20 10>
      constraintSetup {angularLimitsMax=<0 0 25>, angularLimitsMin=<-25 -25 -25>, angularTarget=<0 0 0>}
      localJointOffset=<2 0 0>
     }]
   }
   n30 = Anim Dynamics - Bone: dyn_bow_string_r_03(ComponentPose=@n29, Alpha=@n26, ExternalForce="0,0,0") {
    angularDampingOverride=0.9999
    angularSpringConstant=75
    bAngularSpring=true
    bGravityOverrideInSimSpace=true
    bOverrideAngularDamping=true
    bUseGravityOverride=true
    simulationSpace=RootRelative
    alphaBoolBlend {blendOption=Linear}
    boundBone {boneName=dyn_bow_string_r_03}
    componentLinearAccScale=<1 1 1>
    componentLinearVelScale=<1 1 1>
    gravityOverride {z=-980}
    physicsBodyDefinitions [{
      sphereCollisionRadius=0
      boundBone {boneName=dyn_bow_string_r_03}
      boxExtents=<20 20 10>
      constraintSetup {angularLimitsMax=<25 5 25>, angularLimitsMin=<-25 -25 -25>, angularTarget=<0 0 0>}
      localJointOffset=<2 0 0>
     }]
   }
   Output Pose(Result=Component To Local(ComponentPose=@n30)) {layerGroup=None}
  ]}
 EventGraph {nodes [
   Event BlueprintUpdateAnimation()
   > n1 = Sequence!(execute) -> (then_0, then_1, then_2:Exec, then_3:Exec)
   > n2 = Branch(Condition=Get bIsSwimSprinting() -> (bIsSwimSprinting:bool))
   > Set gravity_override(gravity_override="0, 0, 0")
   n3 = Branch(execute=@n2.else, Condition=Get bIsSkyDiving())
   > n4 = Branch(Condition=Get bIsParachuteOpened())
   Set gravity_override(execute=[@n3.else, @n4.then], gravity_override="0, 0,-980")
   n5 = Branch(execute=@n4.else, Condition="float > float"(A=Get PawnVelocityZ(), B=50) {is=title K2Node_CallFunction})
   > Set gravity_override(gravity_override="0, 0,300")
   Set gravity_override(execute=@n5.else, gravity_override="0, 0,-200")
   n6 = BreakTransform(InTransform=GetDeltaTransformFromRefPose(self=GetOwnerSkeletalMeshComponent(), BoneName=thigh_r, BaseName=pelvis))
   Set thigh rotation_r(execute=@n1.then_1, thigh rotation_r:double=MapRangeClamped(Value=BreakRotator(InRot=@n6.Rotation).Yaw, InRangeA=-50, InRangeB=-90)) -> (Output_Get:double)
   n7 = BreakTransform(InTransform=GetDeltaTransformFromRefPose(self=GetOwnerSkeletalMeshComponent(), BoneName=thigh_l, BaseName=pelvis))
   > Set thigh_rotation_l(thigh_rotation_l:double=MapRangeClamped(Value=BreakRotator(InRot=@n7.Rotation).Yaw, InRangeA=-50, InRangeB=-90)) -> (Output_Get:double)
   > n8 = Branch(Condition=Get bIsJumping())
   Set jumping alpha(execute=[@n9.then, @n8.then], jumping alpha:double=1) -> (Output_Get:double)
   n9 = Branch(execute=@n8.else, Condition=Get bIsFalling())
   Set jumping alpha(execute=@n9.else, jumping alpha=0)
   n10 = Branch(execute=@n1.then_2, Condition=Get bIsCrouchSprinting())
   > Set crouch sprinting(crouch sprinting:double=1) -> (Output_Get:double)
   Set crouch sprinting(execute=@n10.else, crouch sprinting=0)
   n11 = Branch(execute=@n1.then_3, Condition=AND Boolean(A=Get bIsCrouching(), B=Get bIsAccelerating2D()))
   > Set CrouchWalking(CrouchWalking:bool=true) -> (Output_Get:bool)
   Set CrouchWalking(execute=@n11.else, CrouchWalking=false)
   Event OnBeginCrouching()
   > Set crouch alpha(crouch alpha:double=1) -> (Output_Get:double)
   Event OnEndCrouching()
   > Set crouch alpha(crouch alpha=0)
   Event OnCharacterCustomizationLoaded()
   n12 = CheckForBackbling()
   > Set isWearingBackbling(isWearingBackbling=@n12)
   > Set isWearingCape(isWearingCape=@n12.bIsBackblingLong)
  ]}
]
variables [
 crouch alpha {type=number}
 crouch sprinting {type=number}
 CrouchWalking {type=boolean}
 gravity_override {type=Vector}
 isWearingBackbling {type=boolean}
 isWearingCape {type=boolean}
 jumping alpha {type=number}
 thigh rotation_r {type=number}
 thigh_rotation_l {type=number}
]
