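# Physics asset for the F_MED_Myth female player body, setting up secondary motion for the bandana and hair bones.
# Capsules on the upper spine, clavicles, neck and head form the body volumes, joined by constraints with free swing and twist.
# The bandana chain hangs from the head and forks into left and right three-bone tails.
# Single-bone hair and bangs capsules attach to the head with limited swing and very tight twist.
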PhysicsAsset /BRCosmetics/Characters/Player/Female/Medium/Bodies/F_MED_Myth/Meshes/Parts/F_MED_Myth_Dyn
bodies {
 clavicle_l [{shapeType=Capsule, length=7.4062, radius=8.75, center=<10.4426 -1.2937 -1.2158>, rotation=<rot 79.9943 180 180>}]
 clavicle_r [{shapeType=Capsule, length=7.4062, radius=8.75, center=<-10.4426 1.2937 1.2158>, rotation=<rot 79.9943 180 180>}]
 dyn_bandana_01 [{shapeType=Capsule, length=3.6875, radius=1.5, center=<2.5804 0 0>, rotation=<rot 90 0 0>}]
 dyn_bandana_02 [{shapeType=Capsule, length=3.6875, radius=1.4, center=<2.5804 0 0>, rotation=<rot 90 0 0>}]
 dyn_bandana_l_01 [{shapeType=Capsule, length=2.1, radius=1.4, center=<2.5804 0 0>, rotation=<rot 90 0 0>}]
 dyn_bandana_l_02 [{shapeType=Capsule, length=1.5, radius=2.1, center=<2.5804 0 0>, rotation=<rot 90 0 0>}]
 dyn_bandana_l_03 [{shapeType=Capsule, length=3.2188, radius=1.4, center=<2.5804 0 0>, rotation=<rot 90 0 0>}]
 dyn_bandana_r_01 [{shapeType=Capsule, length=2.1, radius=1.4, center=<2.5804 0 0>, rotation=<rot 90 0 0>}]
 dyn_bandana_r_02 [{shapeType=Capsule, length=1.5, radius=2.1, center=<2.5804 0 0>, rotation=<rot 90 0 0>}]
 dyn_bandana_r_03 [{shapeType=Capsule, length=3.2188, radius=1.4, center=<2.5804 0 0>, rotation=<rot 90 0 0>}]
 dyn_bangs_r_01 [{shapeType=Capsule, length=2.8, radius=1.6, center=<2.6 0 0>, rotation=<rot 90 0 0>}]
 dyn_hair_bk_01 [{shapeType=Capsule, length=3, radius=1.5, center=<2.6 0 0>, rotation=<rot 90 0 0>}]
 dyn_hair_bk_l_01 [{shapeType=Capsule, length=3, radius=1.5, center=<2.6 0 0>, rotation=<rot 90 0 0>}]
 dyn_hair_bk_r_01 [{shapeType=Capsule, length=3, radius=1.5, center=<2.6 0 0>, rotation=<rot 90 0 0>}]
 dyn_hair_fr_l_01 [{shapeType=Capsule, length=3, radius=1.5, center=<2.6 0 0>, rotation=<rot 90 0 0>}]
 dyn_hair_fr_r_01 [{shapeType=Capsule, length=3, radius=1.5, center=<2.6 0 0>, rotation=<rot 90 0 0>}]
 head [
  {shapeType=Capsule, length=7.8438, radius=6.0312, center=<2.1081 3.3774 0>, rotation=<rot 90 0 0>}
  {shapeType=Capsule, length=7.8438, radius=6.0312, center=<2.8107 -0.9934 0.0002>, rotation=<rot 90 0 0>}
 ]
 neck_01 [{shapeType=Capsule, length=4.875, radius=7, rotation=<rot 90 0 0>}]
 neck_02 [{shapeType=Capsule, length=4.875, radius=7, rotation=<rot 90 0 0>}]
 spine_05 [
  {shapeType=Capsule, length=12.375, radius=11, center=<4.1536 -1.9595 0>}
  {shapeType=Capsule, length=5.2188, radius=4.5, center=<14.1853 -4.8555 0>}
  {shapeType=Capsule, length=5.6562, radius=5, center=<13.3487 -3.6168 -5.8351>, rotation=<rot -28.8144 5.3777 -44.7565>}
 ]
}
constraints [
 {bone1Name=dyn_bandana_02, bone2Name=dyn_bandana_01, swing1LimitDegrees=20, swing2LimitDegrees=45, twistLimitDegrees=5}
 {bone1Name=dyn_bandana_r_01, bone2Name=dyn_bandana_02, swing1LimitDegrees=25, swing2LimitDegrees=45, twistLimitDegrees=5}
 {bone1Name=dyn_bandana_r_02, bone2Name=dyn_bandana_r_01, swing1LimitDegrees=25, swing2LimitDegrees=45, twistLimitDegrees=5}
 {bone1Name=dyn_bandana_r_03, bone2Name=dyn_bandana_r_02, swing1LimitDegrees=25, swing2LimitDegrees=45, twistLimitDegrees=5}
 {bone1Name=dyn_bandana_l_01, bone2Name=dyn_bandana_02, swing1LimitDegrees=25, swing2LimitDegrees=45, twistLimitDegrees=5}
 {bone1Name=dyn_bandana_l_02, bone2Name=dyn_bandana_l_01, swing1LimitDegrees=25, swing2LimitDegrees=45, twistLimitDegrees=5}
 {bone1Name=dyn_bandana_l_03, bone2Name=dyn_bandana_l_02, swing1LimitDegrees=25, swing2LimitDegrees=45, twistLimitDegrees=5}
 {bone1Name=dyn_hair_fr_l_01, bone2Name=head, swing1LimitDegrees=15, swing2LimitDegrees=45, twistLimitDegrees=1}
 {bone1Name=dyn_hair_bk_l_01, bone2Name=head, swing1LimitDegrees=15, swing2LimitDegrees=45, twistLimitDegrees=1}
 {bone1Name=dyn_hair_bk_01, bone2Name=head, swing1LimitDegrees=15, swing2LimitDegrees=45, twistLimitDegrees=1}
 {bone1Name=dyn_bangs_r_01, bone2Name=head, swing1LimitDegrees=20, swing2LimitDegrees=20, twistLimitDegrees=1}
 {bone1Name=dyn_bandana_01, bone2Name=head, swing1LimitDegrees=20, swing2LimitDegrees=45, twistLimitDegrees=5}
 {bone1Name=dyn_hair_bk_r_01, bone2Name=head, swing1LimitDegrees=15, swing2LimitDegrees=45, twistLimitDegrees=1}
 {bone1Name=dyn_hair_fr_r_01, bone2Name=head, swing1LimitDegrees=15, swing2LimitDegrees=45, twistLimitDegrees=1}
 {
  bone1Name=head
  bone2Name=neck_02
  swing1LimitDegrees=45
  swing1Motion=Free
  swing2LimitDegrees=45
  swing2Motion=Free
  twistLimitDegrees=45
  twistMotion=Free
 }
 {
  bone1Name=neck_02
  bone2Name=neck_01
  swing1LimitDegrees=45
  swing1Motion=Free
  swing2LimitDegrees=45
  swing2Motion=Free
  twistLimitDegrees=45
  twistMotion=Free
 }
 {
  bone1Name=clavicle_l
  bone2Name=spine_05
  swing1LimitDegrees=45
  swing1Motion=Free
  swing2LimitDegrees=45
  swing2Motion=Free
  twistLimitDegrees=45
  twistMotion=Free
 }
 {
  bone1Name=clavicle_r
  bone2Name=spine_05
  swing1LimitDegrees=45
  swing1Motion=Free
  swing2LimitDegrees=45
  swing2Motion=Free
  twistLimitDegrees=45
  twistMotion=Free
 }
 {
  bone1Name=neck_01
  bone2Name=spine_05
  swing1LimitDegrees=45
  swing1Motion=Free
  swing2LimitDegrees=45
  swing2Motion=Free
  twistLimitDegrees=45
  twistMotion=Free
 }
]
