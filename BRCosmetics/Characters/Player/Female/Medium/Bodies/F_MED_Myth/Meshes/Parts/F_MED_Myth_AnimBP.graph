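# Post-copy secondary-motion AnimBP for the F_MED_Myth character part, animating its bandana, hair and bangs.
# It copies the attached parent's pose, runs RigidBody physics with overridden gravity, offsets the bandana bone, then chains spring-driven AnimDynamics on hair bones.
# Each update computes gravity through a deprecated helper and head-to-spine rotation, capped lower when not crouching.
# That rotation drives the bandana offset only while a backbling or cape is worn.
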AnimBlueprint /BRCosmetics/Characters/Player/Female/Medium/Bodies/F_MED_Myth/Meshes/Parts/F_MED_Myth_AnimBP
target_skeleton=&/F_MED_Myth_Skeleton
graphs [
 AnimGraph {nodes [
   n1 = RigidBody!(
    ComponentPose=Local To Component(LocalPose=Copy Pose From Mesh() {bUseAttachedParent=true})
    Alpha
    OverrideWorldGravity:Vector=Get GravityOverride()
    ExternalForce
   ) -> (Pose) {
    bClampLinearTranslationLimitToRefPose=true
    bOverrideWorldGravity=true
    overridePhysicsAsset=&/F_MED_Myth_Dyn
    simulationSpace=BaseBoneSpace
    alphaBoolBlend {blendOption=Linear}
    baseBoneRef {boneName=neck_02}
    componentAppliedLinearAccClamp=<3000 3000 3000>
    componentLinearAccScale=<0.25 0.25 0.15>
    componentLinearVelScale=<0.25 0.25 0.15>
    overrideWorldGravity {z=-980}
    simSpaceSettings {maxAngularAcceleration=500, maxAngularVelocity=500, velocityScaleZ=0, worldAlpha=0.2, externalLinearDragV=<0.1 0.1 0.1>}
   }
   n2 = "Transform (Modify) Bone - Bone: dyn_bandana_01"(
    ComponentPose=@n1
    Alpha=Get HeadRot_Cape_BackPack() -> (HeadRot_Cape_BackPack:double)
    Translation="0,0,2"
    Rotation="40,0,0"
    Scale="1,1,1"
   ) {
    rotationMode=BMM_Additive
    rotationSpace=BCS_BoneSpace
    translationMode=BMM_Additive
    translationSpace=BCS_BoneSpace
    alphaBoolBlend {blendOption=Linear}
    boneToModify {boneName=dyn_bandana_01}
    rotation {pitch=40}
    translation {z=2}
   }
   n3 = Anim Dynamics - Bone: dyn_hair_bk_01(ComponentPose=@n2, Alpha=1, ExternalForce="0,0,0") {
    angularDampingOverride=0.9999
    angularSpringConstant=75
    bAngularSpring=true
    bGravityOverrideInSimSpace=true
    bOverrideAngularDamping=true
    simulationSpace=RootRelative
    alphaBoolBlend {blendOption=Linear}
    boundBone {boneName=dyn_hair_bk_01}
    componentLinearAccScale=<1 1 1>
    componentLinearVelScale=<1 1 1>
    physicsBodyDefinitions [{
      sphereCollisionRadius=0
      boundBone {boneName=dyn_hair_bk_01}
      boxExtents=<20 20 10>
      constraintSetup {angularLimitsMax=<45 0 45>, angularLimitsMin=<-45 -10 -45>, angularTarget=<0 0 0>}
      localJointOffset=<1 0 0>
     }]
   }
   n4 = Anim Dynamics - Bone: dyn_hair_fr_l_01(ComponentPose=@n3, Alpha=1, ExternalForce="0,0,0") {
    angularDampingOverride=0.9999
    angularSpringConstant=75
    bAngularSpring=true
    bGravityOverrideInSimSpace=true
    bOverrideAngularDamping=true
    simulationSpace=RootRelative
    alphaBoolBlend {blendOption=Linear}
    boundBone {boneName=dyn_hair_fr_l_01}
    componentLinearAccScale=<1 1 1>
    componentLinearVelScale=<1 1 1>
    physicsBodyDefinitions [{
      sphereCollisionRadius=0
      boundBone {boneName=dyn_hair_fr_l_01}
      boxExtents=<20 20 10>
      constraintSetup {angularLimitsMax=<45 0 45>, angularLimitsMin=<-45 -45 -45>, angularTarget=<0 0 0>}
      localJointOffset=<1 0 0>
     }]
   }
   n5 = Anim Dynamics - Bone: dyn_hair_bk_l_01(ComponentPose=@n4, Alpha=1, ExternalForce="0,0,0") {
    angularDampingOverride=0.9999
    angularSpringConstant=75
    bAngularSpring=true
    bGravityOverrideInSimSpace=true
    bOverrideAngularDamping=true
    simulationSpace=RootRelative
    alphaBoolBlend {blendOption=Linear}
    boundBone {boneName=dyn_hair_bk_l_01}
    componentLinearAccScale=<1 1 1>
    componentLinearVelScale=<1 1 1>
    physicsBodyDefinitions [{
      sphereCollisionRadius=0
      boundBone {boneName=dyn_hair_bk_l_01}
      boxExtents=<20 20 10>
      constraintSetup {angularLimitsMax=<45 0 45>, angularLimitsMin=<-45 -45 -25>, angularTarget=<0 0 0>}
      localJointOffset=<1 0 0>
     }]
   }
   n6 = Anim Dynamics - Bone: dyn_hair_bk_r_01(ComponentPose=@n5, Alpha=1, ExternalForce="0,0,0") {
    angularDampingOverride=0.9999
    angularSpringConstant=75
    bAngularSpring=true
    bGravityOverrideInSimSpace=true
    bOverrideAngularDamping=true
    simulationSpace=RootRelative
    alphaBoolBlend {blendOption=Linear}
    boundBone {boneName=dyn_hair_bk_r_01}
    componentLinearAccScale=<1 1 1>
    componentLinearVelScale=<1 1 1>
    physicsBodyDefinitions [{
      sphereCollisionRadius=0
      boundBone {boneName=dyn_hair_bk_r_01}
      boxExtents=<20 20 10>
      constraintSetup {angularLimitsMax=<45 0 45>, angularLimitsMin=<-45 -45 -45>, angularTarget=<0 0 0>}
      localJointOffset=<1 0 0>
     }]
   }
   n7 = Anim Dynamics - Bone: dyn_hair_fr_r_01(ComponentPose=@n6, Alpha=1, ExternalForce="0,0,0") {
    angularDampingOverride=0.9999
    angularSpringConstant=75
    bAngularSpring=true
    bGravityOverrideInSimSpace=true
    bOverrideAngularDamping=true
    simulationSpace=RootRelative
    alphaBoolBlend {blendOption=Linear}
    boundBone {boneName=dyn_hair_fr_r_01}
    componentLinearAccScale=<1 1 1>
    componentLinearVelScale=<1 1 1>
    physicsBodyDefinitions [{
      sphereCollisionRadius=0
      boundBone {boneName=dyn_hair_fr_r_01}
      boxExtents=<20 20 10>
      constraintSetup {angularLimitsMax=<45 0 45>, angularLimitsMin=<-45 -45 -45>, angularTarget=<0 0 0>}
      localJointOffset=<1 0 0>
     }]
   }
   n8 = Anim Dynamics - Bone: dyn_bangs_r_01(ComponentPose=@n7, Alpha=1, ExternalForce="0,0,0") {
    angularDampingOverride=0.9999
    angularSpringConstant=75
    bAngularSpring=true
    bGravityOverrideInSimSpace=true
    bOverrideAngularDamping=true
    simulationSpace=RootRelative
    alphaBoolBlend {blendOption=Linear}
    boundBone {boneName=dyn_bangs_r_01}
    componentLinearAccScale=<1 1 1>
    componentLinearVelScale=<1 1 1>
    physicsBodyDefinitions [{
      sphereCollisionRadius=0
      boundBone {boneName=dyn_bangs_r_01}
      boxExtents=<20 20 10>
      constraintSetup {angularLimitsMax=<45 0 45>, angularLimitsMin=<-45 -45 -45>, angularTarget=<0 0 0>}
      localJointOffset=<1 0 0>
     }]
   }
   n9 = Anim Dynamics - Bone: dyn_hair_fr_l_03(ComponentPose=@n8, Alpha=1, ExternalForce="0,0,0") {
    angularDampingOverride=270
    angularSpringConstant=100
    bAngularSpring=true
    bGravityOverrideInSimSpace=true
    linearDampingOverride=270
    simulationSpace=BoneRelative
    alphaBoolBlend {blendOption=Linear}
    boundBone {boneName=dyn_hair_fr_l_03}
    chainEnd {boneName=dyn_hair_fr_l_03}
    componentLinearAccScale=<1 1 1>
    componentLinearVelScale=<1 1 1>
    physicsBodyDefinitions [{
      sphereCollisionRadius=0
      boundBone {boneName=dyn_hair_fr_l_03}
      boxExtents=<10 20 20>
      constraintSetup {angularLimitsMax=<0 5 25>, angularLimitsMin=<0 -5 -25>, angularTarget=<0 0 0>}
      localJointOffset=<1 0 -0.2>
     }]
    relativeSpaceBone {boneName=head}
   }
   n10 = Anim Dynamics - Bone: dyn_hair_bk_l_03(ComponentPose=@n9, Alpha=1, ExternalForce="0,0,0") {
    angularDampingOverride=270
    angularSpringConstant=100
    bAngularSpring=true
    bGravityOverrideInSimSpace=true
    linearDampingOverride=270
    simulationSpace=BoneRelative
    alphaBoolBlend {blendOption=Linear}
    boundBone {boneName=dyn_hair_bk_l_03}
    chainEnd {boneName=dyn_hair_bk_l_03}
    componentLinearAccScale=<1 1 1>
    componentLinearVelScale=<1 1 1>
    physicsBodyDefinitions [{
      sphereCollisionRadius=0
      boundBone {boneName=dyn_hair_bk_l_03}
      boxExtents=<10 20 20>
      constraintSetup {angularLimitsMax=<0 5 45>, angularLimitsMin=<0 -25 0>, angularTarget=<0 0 0>}
      localJointOffset=<1 0 -0.2>
     }]
    relativeSpaceBone {boneName=head}
   }
   n11 = Anim Dynamics - Bone: dyn_hair_bk_03(ComponentPose=@n10, Alpha=1, ExternalForce="0,0,0") {
    angularDampingOverride=270
    angularSpringConstant=100
    bAngularSpring=true
    bGravityOverrideInSimSpace=true
    linearDampingOverride=270
    simulationSpace=BoneRelative
    alphaBoolBlend {blendOption=Linear}
    boundBone {boneName=dyn_hair_bk_03}
    chainEnd {boneName=dyn_hair_bk_03}
    componentLinearAccScale=<1 1 1>
    componentLinearVelScale=<1 1 1>
    physicsBodyDefinitions [{
      sphereCollisionRadius=0
      boundBone {boneName=dyn_hair_bk_03}
      boxExtents=<10 20 20>
      constraintSetup {angularLimitsMax=<0 5 45>, angularLimitsMin=<0 -25 0>, angularTarget=<0 0 0>}
      localJointOffset=<1 0 -0.2>
     }]
    relativeSpaceBone {boneName=head}
   }
   n12 = Anim Dynamics - Bone: dyn_hair_bk_r_03(ComponentPose=@n11, Alpha=1, ExternalForce="0,0,0") {
    angularDampingOverride=270
    angularSpringConstant=100
    bAngularSpring=true
    bGravityOverrideInSimSpace=true
    linearDampingOverride=270
    simulationSpace=BoneRelative
    alphaBoolBlend {blendOption=Linear}
    boundBone {boneName=dyn_hair_bk_r_03}
    chainEnd {boneName=dyn_hair_bk_r_03}
    componentLinearAccScale=<1 1 1>
    componentLinearVelScale=<1 1 1>
    physicsBodyDefinitions [{
      sphereCollisionRadius=0
      boundBone {boneName=dyn_hair_bk_r_03}
      boxExtents=<10 20 20>
      constraintSetup {angularLimitsMax=<0 5 45>, angularLimitsMin=<0 -25 0>, angularTarget=<0 0 0>}
      localJointOffset=<1 0 -0.2>
     }]
    relativeSpaceBone {boneName=head}
   }
   n13 = Anim Dynamics - Bone: dyn_bangs_l_02(ComponentPose=@n12, Alpha=1, ExternalForce="0,0,0") {
    angularDampingOverride=270
    angularSpringConstant=100
    bAngularSpring=true
    bGravityOverrideInSimSpace=true
    linearDampingOverride=270
    simulationSpace=BoneRelative
    alphaBoolBlend {blendOption=Linear}
    boundBone {boneName=dyn_bangs_l_02}
    chainEnd {boneName=dyn_bangs_l_02}
    componentLinearAccScale=<1 1 1>
    componentLinearVelScale=<1 1 1>
    physicsBodyDefinitions [{
      sphereCollisionRadius=0
      boundBone {boneName=dyn_bangs_l_02}
      boxExtents=<10 20 20>
      constraintSetup {angularLimitsMax=<0 5 25>, angularLimitsMin=<0 -5 0>, angularTarget=<0 0 0>}
      localJointOffset=<1 0 -0.2>
     }]
    relativeSpaceBone {boneName=head}
   }
   n14 = Anim Dynamics - Bone: dyn_bangs_r_02(ComponentPose=@n13, Alpha=1, ExternalForce="0,0,0") {
    angularDampingOverride=270
    angularSpringConstant=100
    bAngularSpring=true
    bGravityOverrideInSimSpace=true
    linearDampingOverride=270
    simulationSpace=BoneRelative
    alphaBoolBlend {blendOption=Linear}
    boundBone {boneName=dyn_bangs_r_02}
    chainEnd {boneName=dyn_bangs_r_02}
    componentLinearAccScale=<1 1 1>
    componentLinearVelScale=<1 1 1>
    physicsBodyDefinitions [{
      sphereCollisionRadius=0
      boundBone {boneName=dyn_bangs_r_02}
      boxExtents=<10 20 20>
      constraintSetup {angularLimitsMax=<0 5 0>, angularLimitsMin=<0 -5 -25>, angularTarget=<0 0 0>}
      localJointOffset=<1 0 -0.2>
     }]
    relativeSpaceBone {boneName=head}
   }
   n15 = Anim Dynamics - Bone: dyn_hair_fr_r_03(ComponentPose=@n14, Alpha=1, ExternalForce="0,0,0") {
    angularDampingOverride=270
    angularSpringConstant=100
    bAngularSpring=true
    bGravityOverrideInSimSpace=true
    linearDampingOverride=270
    simulationSpace=BoneRelative
    alphaBoolBlend {blendOption=Linear}
    boundBone {boneName=dyn_hair_fr_r_03}
    chainEnd {boneName=dyn_hair_fr_r_03}
    componentLinearAccScale=<1 1 1>
    componentLinearVelScale=<1 1 1>
    physicsBodyDefinitions [{
      sphereCollisionRadius=0
      boundBone {boneName=dyn_hair_fr_r_03}
      boxExtents=<10 20 20>
      constraintSetup {angularLimitsMax=<0 5 45>, angularLimitsMin=<0 -25 0>, angularTarget=<0 0 0>}
      localJointOffset=<1 0 -0.2>
     }]
    relativeSpaceBone {boneName=head}
   }
   n16 = Anim Dynamics - Bone: dyn_hair_fr_r_04(ComponentPose=@n15, Alpha=1, ExternalForce="0,0,0") {
    angularDampingOverride=270
    angularSpringConstant=100
    bAngularSpring=true
    bGravityOverrideInSimSpace=true
    linearDampingOverride=270
    simulationSpace=BoneRelative
    alphaBoolBlend {blendOption=Linear}
    boundBone {boneName=dyn_hair_fr_r_04}
    chainEnd {boneName=dyn_hair_fr_r_04}
    componentLinearAccScale=<1 1 1>
    componentLinearVelScale=<1 1 1>
    physicsBodyDefinitions [{
      sphereCollisionRadius=0
      boundBone {boneName=dyn_hair_fr_r_04}
      boxExtents=<10 20 20>
      constraintSetup {angularLimitsMax=<0 5 45>, angularLimitsMin=<0 -25 0>, angularTarget=<0 0 0>}
      localJointOffset=<1 0 -0.2>
     }]
    relativeSpaceBone {boneName=head}
   }
   Output Pose(Result=Component To Local(ComponentPose=@n16)) {layerGroup=None}
  ]}
 EventGraph {nodes [
   Event BlueprintUpdateAnimation()
   > n1 = Sequence!(execute) -> (then_0, then_1, then_2:Exec)
   > n2 = Gravity Override State Function DEPRECATED(CharacterPartAnimInstance=Self-Reference(), GameTimeInSeconds=GetGameTimeInSeconds(), GravityOverrideParams=Get Gravity Override Params())
   > Set GravityOverride(GravityOverride=@n2)
   n3 = Branch(execute=@n1.then_1, Condition=Get bIsCrouching())
   n4 = GetAlphaFromReferenceTransformDelta(SkeletalMeshComponent=Get Skel Mesh(), BoneName=spine_05, BaseName=head, InputRange=(X=-15,Y=75), TwistAxis="0,0,1")
   > n5 = Set HeadRot(HeadRot:double="Clamp (Float)"(Value=@n4, Min=0, Max=1)) -> (Output_Get:double)
   > Set HeadRot_Cape_BackPack(HeadRot_Cape_BackPack:double=float * float(A=@n5, B:bool=Get isWaringCapeOrBackPack() -> (isWaringCapeOrBackPack:bool)) {is=title K2Node_PromotableOperator}) -> (Output_Get:double)
   n6 = Set HeadRot(execute=@n3.else, HeadRot="Clamp (Float)"(Value=@n4, Min=0, Max=0.45))
   > Set HeadRot_Cape_BackPack(HeadRot_Cape_BackPack=float * float(A=@n6, B=Get isWaringCapeOrBackPack()))
   n7 = CheckForBackbling()
   Set isWearingBackbling(execute=@n1.then_2, isWearingBackbling=@n7)
   > Set Is Wearing Cape(Is Wearing Cape=@n7.bIsBackblingACape)
   > Set isWaringCapeOrBackPack(isWaringCapeOrBackPack:bool=OR Boolean(A=Get isWearingBackbling(), B=Get Is Wearing Cape() -> (Is Wearing Cape:bool))) -> (Output_Get:bool)
   Event BlueprintInitializeAnimation()
   > Set Skel Mesh(Skel Mesh=GetOwnerSkeletalMeshComponent())
  ]}
]
variables [
 Gravity Override Params {type=GravityOverrideParamsStruct}
 GravityOverride {type=Vector}
 HeadRot {type=number}
 HeadRot_Cape_BackPack {type=number}
 Is Wearing Cape {type=boolean}
 isWaringCapeOrBackPack {type=boolean}
 isWearingBackbling {type=boolean}
 Skel Mesh {type=SkeletalMeshComponent}
]
